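; Conversion function library for TechAudioTools widgets that display and edit MetaSound values.
; GetSourceValue and SetNormalizedValue map between normalized and source ranges, linearly or logarithmically; log normalization clamps the value and returns 0 for near-zero spans.
; SetLiteralWidgetInputViewmodels looks up a literal widget's named input viewmodels on the MetaSound viewmodel and passes the map back to the widget.
; GetMetaSoundDataTypeIcon picks a pin icon name from array and constructor-pin status.
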(fn GetMetaSoundDataTypeIcon
  (params (Data%20Type :type Name) (Is%20Constructor%20Pin :type Boolean))
  (return (select (== Data%20Type "") "EditorStyle.Icons.BulletPoint" (select (Utilities|String|EndsWith (Utilities|String|ToString(Name) Data%20Type) "Array") (select Is%20Constructor%20Pin "MetaSoundStyle.MetasoundEditor.Graph.ConstructorPinArray" "MetaSoundStyle.MetasoundEditor.Graph.ArrayPin") (select Is%20Constructor%20Pin "MetaSoundStyle.MetasoundEditor.Graph.ConstructorPin" "EditorStyle.Icons.BulletPoint")))))

(fn GetSourceValue
  (params (Apply%20Logarithmic%20Interpolation :type Boolean) (Normalized%20Value :type Float%20%28double-precision%29) (Source%20Range%20Min :type Float%20%28double-precision%29) (Source%20Range%20Max :type Float%20%28double-precision%29))
  (if Apply%20Logarithmic%20Interpolation
    (return (Math|Float|Exp (Math|Float|Lerp (Math|Float|Loge (Variables|GetSource%20Range%20Min)) (Math|Float|Loge (Variables|GetSource%20Range%20Max)) (Variables|GetNormalized%20Value))))
    (else
      (return (Math|Float|Lerp (Variables|GetSource%20Range%20Min) (Variables|GetSource%20Range%20Max) (Variables|GetNormalized%20Value))))))

(fn SetLiteralWidgetInputViewmodels
  (params (Literal%20Widget :type Meta%20Sound%20Literal%20Widget%20Interface%20Interface) (MetaSound%20Viewmodel :type Meta%20Sound%20Viewmodel%20Object%20Reference))
  (for _array_element (MetaSoundLiteral|GetInputViewmodelNames Literal%20Widget)
    (Utilities|Map|Add (LocalVariables|GetInput%20Viewmodels) _array_element (MetaSoundViewmodel|FindInputViewmodel (Variables|GetMetaSound%20Viewmodel) _array_element)))
  (MetaSoundLiteral|SetInputViewmodels (Variables|GetLiteral%20Widget) (LocalVariables|GetInput%20Viewmodels)))

(fn SetNormalizedValue
  (params (Apply%20Logarithmic%20Interpolation :type Boolean) (Source%20Value :type Float%20%28double-precision%29) (Source%20Range%20Min :type Float%20%28double-precision%29) (Source%20Range%20Max :type Float%20%28double-precision%29))
  (if Apply%20Logarithmic%20Interpolation
    (return (select (or (<= (Math|Float|Absolute(Float) (- (Math|Float|Loge (Math|Float|Max(Float) (Variables|GetSource%20Range%20Min) (Variables|GetSource%20Range%20Max))) (Math|Float|Loge (Math|Float|Min(Float) (Variables|GetSource%20Range%20Min) (Variables|GetSource%20Range%20Max))))) 0.0001) (<= (Math|Float|Absolute(Float) (- (Math|Float|Max(Float) (Variables|GetSource%20Range%20Max) 0.0001) (Math|Float|Max(Float) (Variables|GetSource%20Range%20Min) 0.0001))) 0.0001)) 0.0 (/ (- (Math|Float|Loge (Math|Float|Max(Float) (Math|Float|Clamp(Float) (Variables|GetSource%20Value) (Math|Float|Min(Float) (Variables|GetSource%20Range%20Min) (Variables|GetSource%20Range%20Max)) (Math|Float|Max(Float) (Variables|GetSource%20Range%20Min) (Variables|GetSource%20Range%20Max))) 0.0001)) (Math|Float|Loge (Math|Float|Min(Float) (Variables|GetSource%20Range%20Min) (Variables|GetSource%20Range%20Max)))) (- (Math|Float|Loge (Math|Float|Max(Float) (Variables|GetSource%20Range%20Min) (Variables|GetSource%20Range%20Max))) (Math|Float|Loge (Math|Float|Min(Float) (Variables|GetSource%20Range%20Min) (Variables|GetSource%20Range%20Max)))))))
    (else
      (return (Math|Float|NormalizetoRange (Variables|GetSource%20Value) (Variables|GetSource%20Range%20Min) (Variables|GetSource%20Range%20Max))))))
